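; Rocket Ram gameplay ability that applies a predicted braking friction effect to the player and raises movement net acceleration tolerance while active.
; Ram state changes, a one-second delay or gaining the sliding tag trigger suppression and removal.
; Suppression adds loose tags on client and server to ignore the effect, since removal cannot be predicted, while authority removes the effects by granted tag.
; Ending clears the task, binding, tolerance and tags.
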(event Ability|EventActivateAbility
  (|CallParentFunction)
  ;[0.09279 1 0.4823] Apply from the GA to get a client prediction key.
  (Ability|ApplyGameplayEffectToOwner "/RocketRamGameplay/Items/RocketRam/Effects/BrakingFriction/GE_BrakingFriction_RocketRam.GE_BrakingFriction_RocketRam_C")
  (bind _outfortplayerpawn (CallFunction|GetAvatarasValidFortPlayerPawn)
    (:ValidFortPlayerPawn
      (Variables|Default|SetPlayerPawn _outfortplayerpawn)
      (bind _asb_weapon_rocket_ram (Utilities|Casting|CastToB_Weapon_RocketRam (Weapon|GetCurrentWeapon _outfortplayerpawn)))
      (bind _output_get (Variables|Default|SetWeapon _asb_weapon_rocket_ram))
      (Default|BindEventtoOnRamStateChangedEvent _output_get (EventDispatchers|CreateEvent :!event Custom|OnRamStateChanged))
      (bind _asynctaskproxy (Ability|Tasks|WaitGameplayTagAdd "(TagName=\"MovementMode.Sliding\")" 0 true)
        (:then
          (Variables|Default|SetWaitSlidingAsyncTask _asynctaskproxy)
          (bind _asfort_movement_comp_character (Utilities|Casting|CastToFortMovementComp_Character (Variables|/Script/Engine.Character|GetCharacterMovement (Variables|Default|GetPlayerPawn))))
          (NetTolerance|SetUseHighAccelerationTolerance _asfort_movement_comp_character true "RocketRam_BrakingFriction")
          (Class|RocketRamSharedLibrary|RamLog (Variables|Default|GetPlayerPawn) (Utilities|Text|FormatText "NSLOCTEXT(\"[14E2F367E94CDA261D12F120E4C0B819]\", \"A2042A3E4F56735413A30FA6492A994E\", \"Braking Friction - Activated.\")") false true)
          (CallFunction|OnRamStateChanged (Variables|/RocketRamGameplay/Items/RocketRam/B_Weapon_RocketRam.B_Weapon_RocketRam_C|GetRamState (Variables|Default|GetWeapon))))
        (:Added
          (CallFunction|SuppressandRemoveBrakingFriction))))))

(event Ability|EventFailedToActivatePassiveAbility
  (params (ActorInfo :type Gameplay%20Ability%20Actor%20Info%20Structure))
  (|CallParentFunction ActorInfo))

(event Custom|OnRamStateChanged
  (params (RamState :type Enum_RocketRamState%20Enum :default "NewEnumerator0"))
  (switch Utilities|FlowControl|Switch|SwitchonEnum_RocketRamState RamState
    (:NewEnumerator3
      _)
    (:NewEnumerator7
      (CallFunction|SuppressandRemoveBrakingFriction))
    (:NewEnumerator4
      _))
  (Utilities|FlowControl|Delay 1.0)
  (CallFunction|SuppressandRemoveBrakingFriction))

(event Custom|SuppressAndRemoveBrakingFriction
  (bind _brakingfrictionsuppresstags (Variables|Default|GetBrakingFrictionSuppressTags))
  (Class|RocketRamSharedLibrary|RamLog (Variables|Default|GetPlayerPawn) (Utilities|Text|FormatText "NSLOCTEXT(\"[14E2F367E94CDA261D12F120E4C0B819]\", \"0F725FF8411D2483F01DE5846FF9F2AE\", \"Braking Friction - Suppressed and removed.\")") false true)
  ;[1 0.3037 0.8964] Ignore the Movement GE on the server and client. This makes it so the GE removal can be replicated since there's no way to predictively remove GEs.
  (bind _playerpawn (Variables|Default|GetPlayerPawn)
    (:then
      (if (not (GameplayTags|HasAnyMatchingGameplayTags _playerpawn _brakingfrictionsuppresstags))
        (Ability|GameplayEffect|AddLooseGameplayTags _playerpawn _brakingfrictionsuppresstags))
      (Ability|HasAuthority
        (:True
          ;[0 1 0.7754] Removes the Grant GE and the Movement GE.
          (Ability|RemoveGameplayEffectFromOwnerWithGrantedTags (Variables|Default|GetBrakingFrictionTags)))))
    (:else)))

(event Ability|EventOnEndAbility
  (params (bWasCancelled :type Boolean))
  (bind _brakingfrictionsuppresstags (Variables|Default|GetBrakingFrictionSuppressTags))
  (|CallParentFunction bWasCancelled)
  (Class|RocketRamSharedLibrary|RamLog (Variables|Default|GetPlayerPawn) (Utilities|Text|FormatText "NSLOCTEXT(\"[14E2F367E94CDA261D12F120E4C0B819]\", \"414BC4C94834320338847692FF4E5A7C\", \"Braking Friction - Ended.\")") false true)
  (Utilities|FlowControl|Sequence
    (:then_0
      (bind _waitslidingasynctask (Variables|Default|GetWaitSlidingAsyncTask)
        (:then
          (GameplayTasks|EndTask _waitslidingasynctask))
        (:else)))
    (:then_1
      (bind _weapon (Variables|Default|GetWeapon)
        (:then
          (Default|UnbindEventfromOnRamStateChangedEvent _weapon (EventDispatchers|CreateEvent :!event Custom|OnRamStateChanged)))
        (:else)))
    (:then_2
      (bind _playerpawn (Variables|Default|GetPlayerPawn)
        (:then
          (bind _asfort_movement_comp_character (Utilities|Casting|CastToFortMovementComp_Character (Variables|/Script/Engine.Character|GetCharacterMovement _playerpawn)))
          (NetTolerance|SetUseHighAccelerationTolerance _asfort_movement_comp_character false "RocketRam_BrakingFriction"))
        (:else)))
    (:then_3
      (bind _playerpawn_2 (Variables|Default|GetPlayerPawn)
        (:then
          (if (GameplayTags|HasAnyMatchingGameplayTags _playerpawn_2 _brakingfrictionsuppresstags)
            (Ability|GameplayEffect|RemoveLooseGameplayTags _playerpawn_2 _brakingfrictionsuppresstags)))
        (:else)))))
